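# Animation Blueprint for the female medium Sailor Squad Leader head cosmetic, targeting the Fortnite base head skeleton.
# The update event defers to the parent class.
# Event graph logic validates the part's skeletal mesh, waits a zero-duration Delay, then writes the earpiece socket location to the HeadSocketLocation material vector parameter, exposing head position to the shaders.
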AnimBlueprint /BRCosmetics/Characters/Player/Female/Medium/Heads/F_MED_Sailor_Squad_Leader_Head/Meshes/F_MED_Sailor_Squad_Leader_Head_AnimBP
target_skeleton=/Game/Characters/Player/Common/Fortnite_Base_Head/Fortnite_Base_Head_Export_Skeleton
graphs [EventGraph {nodes [
   n1 = Event BlueprintUpdateAnimation()
   > Parent: BlueprintUpdateAnimation!(execute, DeltaTimeX:float=@n1.DeltaTimeX) -> (then) {is=title K2Node_CallParentFunction}
   Event BlueprintPostEvaluateAnimation() {is=title K2Node_Event}
   n2 = GetPartSkeletalMeshComponent()
   > Is Valid(InputObject=@n2)
   > Delay(Duration=0)
   > SetVectorParameterValueOnMaterials!(
    execute
    self:Mesh Component Object Reference=@n2
    ParameterName:Name=HeadSocketLocation
    ParameterValue:Vector=GetSocketLocation!(self:Scene Component Object Reference=@n2, InSocketName:Name=earpiece) -> (ReturnValue:Vector)
   ) -> (then)
   TryGetPawnOwner()
  ]}]
variables []
